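# Niagara burst effect for the Standoff emote's Builder variant, apparently played as the prop spawns.
# A brief air glow joins smoke, curl-noise sparkles, twinkles and debris, mostly spawned on a tall, narrow, tilted cylinder's surface.
# Debris pairs a few colliding mesh chunks with cheap GPU sprite bits.
# Every emitter tints by Fortnite time of day, while smoke and drifting sparkles also adjust color for the front end.
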NiagaraSystem /BRCosmetics/Effects/Fort_Effects/Emotes/Standoff/NS_Standoff_Emote_Builder_SpawnProp
dependencies {
 usedDataInterfaces [
  &NiagaraDataInterfaceCurve
  &NiagaraDataInterfaceSpriteRendererInfo
  &NiagaraDataInterfaceColorCurve
 ]
 usedDynamicInputs [&RandomRangeVector2D, &Add_Vector]
 usedModules [&SystemState]
}
emitters {
 AirGlow {
  emitterUpdateScript [
   &EmitterState()
   &SpawnBurst_Instantaneous(Spawn Count=1, Use Spawn Probability=false)
  ]
  particleSpawnScript [
   /Niagara/Modules/Spawn/Initialization/InitializeParticle(
    Color=<rgba 0.4323 0.4323 0.4323 0.2>
    Lifetime=0.25
    Module.Write Color=true
    Module.Write Lifetime=true
    Module.Write Mass=false
    Module.Write Material Random=true
    Module.Write Position=false
    Module.Write Scale=false
    Module.Write SpriteRotation=false
    Module.Write SpriteSize=true
    Position=&SimulationPosition()
    Sprite Size=&Vector2DFromFloat(Value=120)
   )
   &DynamicMaterialParameters(
    Index 0 Param 1=2
    Module.Param0WriteEnabled=true
    Module.Param1WriteEnabled=true
    Module.Param2WriteEnabled=true
    Module.Param3WriteEnabled=true
   )
   &CameraOffset(Camera Offset Amount=10)
   &SystemLocation(MaskSpawnGroup=false)
  ]
  particleUpdateScript [
   &ParticleState()
   &ScaleSpriteSize(Scale Factor=&Vector2DFromFloat(Value=&FloatFromCurve(FloatCurve=DataInterface() {propertyValues {Curve {keys [<at 0 1 RCIM_Cubic>]}}})))
   &FN_Light_Attributes(
    Color=$Particles.Color
    Create Color Attribute=true
    Create Exponent Attribute=true
    Create Radius Attribute=true
    Exponent=2
    Radius=200
    Scale Color=&FloatFromCurve(FloatCurve=DataInterface() {propertyValues {Curve {keys [<at 0 0>,<at 0.2 1>,<at 1 0>]}}}, Scale Curve=10)
   )
   &ScaleColor(
    Scale Alpha=&FloatFromCurve(FloatCurve=DataInterface() {propertyValues {Curve {keys [<at 0 0>,<at 0.25 0.75>,<at 1 0>]}}})
    ScaleA=true
    ScaleRGB=true
   ) {module=ScaleColor001}
   &self:ColorFromTOD_09(
    Afternoon:LinearColor=<rgba 1 0.8842 0.8294 1>
    Evening:LinearColor=<rgba 1 0.625 0.5 1>
    Morning:LinearColor=<rgba 1 0.75 0.7753 1>
    Night:LinearColor=<rgba 0.7 0.3494 0.4483 1>
    TOD:Vector4f=&MakeVector4_FromLinearColor(LinearColor=$NPC.FortniteNPC.FortniteActiveTimeOfDay)
   ) {module=ColorFromTOD_09}
  ]
  rendererClasses [&NiagaraSpriteRendererProperties]
  renderers [{rendererClass=&NiagaraSpriteRendererProperties}]
 }
 Chunks_Large {
  emitterUpdateScript [
   &EmitterState()
   &SpawnBurst_Instantaneous(Spawn Count=4, Use Spawn Probability=false, UseLoopCountLimit=false)
  ]
  particleSpawnScript [
   /Niagara/Modules/Spawn/Initialization/V2/InitializeParticle(
    Color Mode=::NewEnumerator0
    Lifetime Max=1.5
    Lifetime Mode=::NewEnumerator1
    Mass Mode=::NewEnumerator1
    Mesh Scale Mode=::NewEnumerator3
    UsePositionOffset=false
    Write Material Random=true
   )
   &ShapeLocation(
    Cylinder Height=150
    Cylinder Radius=10
    MaskSpawnGroup=false
    Offset=<0 0 -5>
    Offset Mode=::NewEnumerator0
    Override Seed=false
    Rotation Mode=::NewEnumerator2
    Shape Primitive=::NewEnumerator1
    Surface Only=true
    UseOwnerScale=true
    Yaw / Pitch / Roll=<0 -0.2 -0.05>
   )
   &InitialMeshOrientation(Use Rotation Vector=false)
   /Niagara/Modules/Update/Mass/CalculateSizeAndRotationalInertiaByMass(
    Calculate Mesh Scale=true
    Density by Material Type:ENiagaraMassByVolume=::NewEnumerator3
    Initial Model Dimensions:Vector3f=&RandomRangeVector(Minimum=<40 40 40>, Maximum=<20 20 20>)
    Module.Manually Enter Density=false
   )
   &AddVelocity(
    Cone Axis=<0 0 1>
    Module.UseConeFalloff=false
    Module.UseVelDistribution=false
    Override Seed=false
    Velocity Mode=::NewEnumerator2
    Velocity Speed=200
   )
   &MeshRotationForce(Rotation=&RandomRangeVector(Minimum=<-0.03 -0.03 -0.03>, Maximum=<0.03 0.03 0.03>))
   &ApplyInitialForces()
  ]
  particleUpdateScript [
   &ParticleState()
   &ScaleMeshSize(Scale Factor=&Multiply_VectorByFloat(
     Vector=<1 1 1>
     Float=&FloatFromCurve(
      FloatCurve=DataInterface() {propertyValues {Curve {keys [
          {interpMode=RCIM_Cubic, tangentMode=RCTM_User, time=0, value=0}
          <at 0.2 1>
          {interpMode=RCIM_Cubic, tangentMode=RCTM_User, time=0.8, value=1}
          {tangentMode=RCTM_User, time=1, value=0}
         ]}}}
      Scale Curve=&RandomRangeFloat(Minimum=0.25, Maximum=0.5)
     )
    ))
   &GravityForce()
   &Drag(Drag=0, Ignore Mass=false, Rotational Drag=3, Use Linear Drag=true, Use Rotational Drag=true)
   &Collision(
    Age Colliding Particles=false
    Collision Enabled=&SetBoolByFloatComparison(A=$Particles.Lifetime, B=0.1)
    Control Roll On Collision=true
    Friction=0.6
    Friction During a Bounce=0.1
    KillOnCollision=false
    Mesh Dimensions:Vector3f=$Particles.Scale
    Module.Control Rest State Via Surface Normal=false
    Module.EnableMaxCollisionCount=false
    Module.Randomize Collision Normal Vector=false
    Radius Calculation Type=::NewEnumerator1
    Radius Scale for Rotation=30
    Restitution=0.3
    ScaleRateOfRotation=true
   )
   &SolveForcesAndVelocity(Clamp Velocity=false, Limit Acceleration=false)
   &self:ColorFromTOD_09(
    Afternoon:LinearColor=<rgba 1 0.8842 0.8294 1>
    Evening:LinearColor=<rgba 1 0.625 0.5 1>
    Morning:LinearColor=<rgba 1 0.75 0.7753 1>
    Night:LinearColor=<rgba 0.7 0.3494 0.4483 1>
    TOD:Vector4f=&MakeVector4_FromLinearColor(LinearColor=$NPC.FortniteNPC.FortniteActiveTimeOfDay)
   ) {module=ColorFromTOD_09}
  ]
  rendererClasses [&NiagaraMeshRendererProperties]
  renderers [{rendererClass=&NiagaraMeshRendererProperties, rendererIndex=0}]
 }
 GPU_Chunks_Cheap {
  simTarget=GPUComputeSim
  emitterUpdateScript [
   &EmitterState()
   &SpawnBurst_Instantaneous(
    Spawn Count=&RandomRangeInt(Minimum=5, Maximum=10)
    Use Spawn Probability=false
    UseLoopCountLimit=false
   )
  ]
  particleSpawnScript [
   &InitializeParticle(
    Lifetime Max=1.5
    Lifetime Min=0.25
    Lifetime Mode=::NewEnumerator1
    Mass Mode=::NewEnumerator1
    Sprite Rotation Mode=::NewEnumerator2
    Sprite Size Mode=::NewEnumerator1
    Uniform Sprite Size Max=8
    Uniform Sprite Size Min=4
    UsePositionOffset=false
    Write Material Random=true
   )
   &ShapeLocation(
    Cylinder Height=150
    Cylinder Radius=10
    MaskSpawnGroup=false
    Offset=<0 0 -5>
    Offset Mode=::NewEnumerator0
    Override Seed=false
    Rotation Mode=::NewEnumerator2
    Shape Primitive=::NewEnumerator1
    Surface Only=true
    UseOwnerScale=true
    Yaw / Pitch / Roll=<0 -0.2 -0.05>
   )
   &AddVelocity(
    Cone Axis=<0 0 1>
    Module.UseConeFalloff=false
    Module.UseVelDistribution=false
    Override Seed=false
    Velocity Mode=::NewEnumerator2
    Velocity Speed=200
   )
  ]
  particleUpdateScript [
   &GravityForce()
   &Drag(Drag=0.25, Use Linear Drag=true, Use Rotational Drag=true)
   &ParticleState()
   &SpriteRotationRate(Rotation Rate=&RandomRangeFloat(Minimum=450, Maximum=&RandomRangeFloat(Minimum=-1, Maximum=1)))
   &ScaleSpriteSize(Uniform Curve Sprite Scale=DataInterface() {propertyValues {Curve {keys [
        {arriveTangent=9.0725, interpMode=RCIM_Cubic, leaveTangent=9.0725, tangentMode=RCTM_User, time=0, value=0}
        {arriveTangent=-2.5249, interpMode=RCIM_Cubic, leaveTangent=-2.5249, tangentMode=RCTM_User, time=0.1408, value=0.9847}
        {arriveTangent=-0.0405, interpMode=RCIM_Cubic, leaveTangent=-0.0405, tangentMode=RCTM_User, time=1, value=0}
       ]}}})
   &SolveForcesAndVelocity(Clamp Velocity=false, Limit Acceleration=false)
   &ScaleColor(Scale Mode=::NewEnumerator2)
   &self:ColorFromTOD_09(
    Afternoon:LinearColor=<rgba 0.3137 0.5098 0.6588 1>
    Evening:LinearColor=<rgba 1 0.625 0.5 1>
    Morning:LinearColor=<rgba 1 0.7396 0.7659 1>
    Night:LinearColor=<rgba 0.7 0.3494 0.4483 1>
    TOD:Vector4f=&MakeVector4_FromLinearColor(LinearColor=$NPC.FortniteNPC.FortniteActiveTimeOfDay)
   ) {module=ColorFromTOD_09}
  ]
  rendererClasses [&NiagaraSpriteRendererProperties]
  renderers [{rendererClass=&NiagaraSpriteRendererProperties}]
 }
 Smoke001 {
  emitterUpdateScript [
   &EmitterState()
   &SpawnBurst_Instantaneous(Spawn Count=5, Spawn Group=2, Use Spawn Probability=false, UseLoopCountLimit=false)
  ]
  particleSpawnScript [
   &InitializeParticle(
    Color=<rgba 0.4043 0.4883 0.5255 0.35>
    Lifetime Max=1.25
    Lifetime Min=0.5
    Lifetime Mode=::NewEnumerator1
    Mass Mode=::NewEnumerator1
    Sprite Rotation Mode=::NewEnumerator2
    Sprite Size=&Multiply_Vector2D_ByFloat(
     Float=2
     Vector2D=&Vector2DFromFloat(Value=&UniformRangedFloat(Minimum=25, Maximum=35))
    )
    Sprite Size Mode=::NewEnumerator4
    UsePositionOffset=false
    Write Material Random=true
   ) {module=InitializeParticle001}
   &AccelerationForce(Acceleration=&UniformRangedVector(Minimum=<0 0 -25>, Maximum=<0 0 -75>))
   &ShapeLocation(
    Cylinder Height=150
    Cylinder Radius=10
    MaskSpawnGroup=false
    Override Seed=false
    Rotation Mode=::NewEnumerator2
    Shape Primitive=::NewEnumerator1
    Surface Only=true
    Surface Only Band Thickness=0.25
    UseOwnerScale=true
    Yaw / Pitch / Roll=<0 -0.2 -0.05>
   )
   &ApplyInitialForces()
   &CameraOffset(Camera Offset Amount=&RandomRangeFloat(Minimum=0, Maximum=15)) {module=CameraOffset001}
   &AddVelocity(Velocity=&UniformRangedVector(Minimum=<-100 -100 -100>, Maximum=<100 100 100>))
   /SetVariables(Particles.UVScale:Vector2f=&UniformRangedVector2D(Minimum=<-1 1>, Maximum=<-1 1>)) {isSetParameters=true}
  ]
  particleUpdateScript [
   &ParticleState()
   &Drag(Drag=&UniformRangedFloat(Minimum=8, Maximum=16), Use Linear Drag=true, Use Rotational Drag=false)
   &GravityForce(Gravity=<0 0 -420>)
   &SolveForcesAndVelocity(Clamp Velocity=false, Limit Acceleration=false)
   &SpriteRotationRate(Rotation Rate=&Multiply_Float(
     A=&Lerp_Float(A=400, B=-400, Alpha=$Particles.Initial.MaterialRandom)
     B=&Lerp_Float(A=0.15, B=0.05, Alpha=$Particles.NormalizedAge)
    ))
   &ScaleColor(
    Scale Alpha=&FloatFromCurve(
     FloatCurve=DataInterface() {propertyValues {Curve {keys [
         {arriveTangent=-6.8822, interpMode=RCIM_Cubic, leaveTangent=-6.8822, tangentMode=RCTM_User, time=0, value=1}
         <at 0.1011 0.2206>
         {arriveTangent=0.137, interpMode=RCIM_Cubic, leaveTangent=0.137, tangentMode=RCTM_User, time=0.75, value=0}
         <at 0.75 0>
        ]}}}
     Scale Curve=1.25
    )
    Scale RGB=&FN_FrontEndVector3Modifier(Gameplay=<1.25 1.25 1.25>, Front End=<0.7 0.9 1.1>)
    ScaleA=true
    ScaleRGB=true
   )
   &ScaleSpriteSize(Scale Factor=&Vector2DFromFloat(Value=&Lerp_Float(A=1, B=2, Alpha=$Particles.NormalizedAge)))
   &SubUVAnimation(
    "Sprite Renderer "=DataInterface() {propertyValues {SpriteRenderer=&self:Smoke001_2.NiagaraSpriteRendererProperties_0}}
    SubUV Animation Mode:ENiagara_SubUVLookupModeV2=::NewEnumerator1
    UseEndFrame=false
    UseStartFrame=false
   ) {module=SubUVAnimation001}
   &self:ColorFromTOD_09(
    Afternoon:LinearColor=<rgba 0.9994 1 0.9635 2>
    Evening:LinearColor=<rgba 0.7135 0.3446 0.1524 1>
    Morning:LinearColor=<rgba 1 0.6075 0.5729 1>
    Night:LinearColor=<rgba 0.7 0.3494 0.4483 1>
    TOD:Vector4f=&MakeVector4_FromLinearColor(LinearColor=$NPC.FortniteNPC.FortniteActiveTimeOfDay)
   ) {module=ColorFromTOD_09}
  ]
  rendererClasses [&NiagaraSpriteRendererProperties]
  renderers [{rendererClass=&NiagaraSpriteRendererProperties}]
 }
 Sparkles {
  emitterUpdateScript [
   &EmitterState()
   &SpawnBurst_Instantaneous(Spawn Count=12, Use Spawn Probability=false)
  ]
  particleSpawnScript [
   &InitializeParticle(
    Color=<rgba 5 3.7043 1.51 1>
    Lifetime=&UniformRangedFloat(Minimum=0.3, Maximum=1)
    Mass Mode=::NewEnumerator1
    Sprite Size Mode=::NewEnumerator1
    Uniform Sprite Size Max=30
    Uniform Sprite Size Min=15
    UsePositionOffset=false
    Write Material Random=true
   ) {module=InitializeParticle001}
   &ShapeLocation(
    Cylinder Height=150
    Cylinder Radius=10
    MaskSpawnGroup=false
    Offset=<0 0 -5>
    Offset Mode=::NewEnumerator0
    Override Seed=false
    Rotation Mode=::NewEnumerator2
    Shape Primitive=::NewEnumerator1
    Surface Only=true
    UseOwnerScale=true
    Yaw / Pitch / Roll=<0 -0.2 -0.05>
   )
   &ApplyInitialForces()
   &AddVelocity(Velocity=&UniformRangedVector(Minimum=<-100 -100 -100>, Maximum=<100 100 100>)) {module=AddVelocity001}
  ]
  particleUpdateScript [
   &ParticleState()
   &GravityForce(Gravity=<0 0 100>)
   /Niagara/Modules/Update/Forces/CurlNoiseForce(Noise Frequency=0.02, Noise Strength=500, Pan Noise=false)
   &Drag(Drag=5, Use Linear Drag=true, Use Rotational Drag=false)
   &SolveForcesAndVelocity(Clamp Velocity=false, Limit Acceleration=false)
   &ScaleColor(
    Scale Alpha=&FloatFromCurve(FloatCurve=DataInterface() {propertyValues {Curve {keys [<at -0.0092 0.0144>,<at 0.0967 0.9767>,<at 1.0011 0.002>]}}})
    Scale RGB=&VectorFromFloat(Value=&FN_FrontEndFloatModifier(Gameplay=1, Front End=0.5))
    ScaleA=true
    ScaleRGB=true
   )
   &ScaleSpriteSize(Scale Factor=&Vector2DFromFloat(Value=&FloatFromCurve(FloatCurve=DataInterface() {propertyValues {Curve {keys [<at -0.0046 0>,<at 0.1746 0.5>,<at 1.0011 0.002>]}}})))
   &self:ColorFromTOD_09(
    Afternoon:LinearColor=<rgba 1 0.8842 0.8294 1>
    Evening:LinearColor=<rgba 1 0.625 0.5 1>
    Morning:LinearColor=<rgba 1 0.75 0.7753 1>
    Night:LinearColor=<rgba 0.7 0.3494 0.4483 1>
    TOD:Vector4f=&MakeVector4_FromLinearColor(LinearColor=$NPC.FortniteNPC.FortniteActiveTimeOfDay)
   ) {module=ColorFromTOD_09}
  ]
  rendererClasses [&NiagaraSpriteRendererProperties]
  renderers [{rendererClass=&NiagaraSpriteRendererProperties}]
 }
 Sparkles_Back {
  emitterUpdateScript [
   &EmitterState()
   &SpawnBurst_Instantaneous(Spawn Count=3, Use Spawn Probability=false, UseLoopCountLimit=false)
   &SpawnBurst_Instantaneous(Spawn Count=3, Spawn Time=0.1, Use Spawn Probability=false, UseLoopCountLimit=false) {module=SpawnBurst_Instantaneous001}
  ]
  particleSpawnScript [
   &InitializeParticle(
    Color=<rgba 0.5 1 0.9833 1>
    Lifetime Max=0.75
    Lifetime Min=0.25
    Lifetime Mode=::NewEnumerator1
    Mass Min=0.675
    Mass Mode=::NewEnumerator3
    Sprite Rotation Mode=::NewEnumerator2
    Sprite Size Mode=::NewEnumerator1
    Uniform Sprite Size Max=35
    Uniform Sprite Size Min=15
    UsePositionOffset=false
    Write Material Random=true
   )
   &ShapeLocation(
    Cylinder Height=150
    Cylinder Radius=25
    MaskSpawnGroup=false
    Offset=<0 0 -25>
    Offset Mode=::NewEnumerator0
    Override Seed=false
    Rotation Mode=::NewEnumerator2
    Shape Primitive=::NewEnumerator1
    Surface Only=true
    UseOwnerScale=true
    Yaw / Pitch / Roll=<0 -0.2 -0.05>
   )
   &CameraOffset(Camera Offset Amount=10)
  ]
  particleUpdateScript [
   &ParticleState()
   &ScaleColor(
    Scale Alpha=&FloatFromCurve(FloatCurve=DataInterface() {propertyValues {Curve {keys [<at -0.0046 1.0113 RCIM_Cubic>,<at 1.0011 0.002 RCIM_Cubic>]}}})
    ScaleA=true
    ScaleRGB=true
   )
   &ScaleColor(
    Scale Alpha=&FloatFromCurve(FloatCurve=DataInterface() {propertyValues {Curve {keys [
         {arriveTangent=5.883, interpMode=RCIM_Cubic, leaveTangent=5.883, tangentMode=RCTM_User, time=0, value=0}
         <at 0.5 1 RCIM_Cubic>
         {arriveTangent=-5.491, interpMode=RCIM_Cubic, leaveTangent=-5.491, tangentMode=RCTM_User, time=1, value=0}
        ]}}})
    ScaleA=true
    ScaleRGB=true
   ) {module=ScaleColor001}
   &ScaleSpriteSize(Uniform Curve Sprite Scale=DataInterface() {propertyValues {Curve {keys [
        {interpMode=RCIM_Cubic, leaveTangentWeight=0.8281, tangentMode=RCTM_User, time=0, value=0}
        {arriveTangent=5.598, interpMode=RCIM_Cubic, tangentMode=RCTM_Break, time=0.5, value=1}
        {arriveTangent=-5.9908, interpMode=RCIM_Cubic, leaveTangent=-5.9908, tangentMode=RCTM_User, time=1, value=0}
       ]}}})
   &SpriteRotationRate(Rotation Rate=&RandomRangeFloat(Minimum=-90, Maximum=90))
   &self:ColorFromTOD_09(
    Afternoon:LinearColor=<rgba 1 0.8842 0.8294 1>
    Evening:LinearColor=<rgba 1 0.625 0.5 1>
    Morning:LinearColor=<rgba 1 0.75 0.7753 1>
    Night:LinearColor=<rgba 0.7 0.3494 0.4483 1>
    TOD:Vector4f=&MakeVector4_FromLinearColor(LinearColor=$NPC.FortniteNPC.FortniteActiveTimeOfDay)
   ) {module=ColorFromTOD_09}
  ]
  rendererClasses [&NiagaraSpriteRendererProperties]
  renderers [{rendererClass=&NiagaraSpriteRendererProperties}]
 }
}
summary {
 systemName=NS_Standoff_Emote_Builder_SpawnProp
 emitters [
  {emitterName=AirGlow, rendererClasses [&NiagaraSpriteRendererProperties]}
  {emitterName=Sparkles, rendererClasses [&NiagaraSpriteRendererProperties]}
  {emitterName=Smoke001, rendererClasses [&NiagaraSpriteRendererProperties]}
  {emitterName=Sparkles_Back, rendererClasses [&NiagaraSpriteRendererProperties]}
  {emitterName=GPU_Chunks_Cheap, simTarget=GPUComputeSim, rendererClasses [&NiagaraSpriteRendererProperties]}
  {emitterName=Chunks_Large, rendererClasses [&NiagaraMeshRendererProperties]}
 ]
 userVariables [
  User.bIsFrontEnd {type=NiagaraBool}
  User.bIsFrontEndPreview {type=NiagaraBool}
  User.bIsInWater {type=NiagaraBool}
  User.Boost {type=NiagaraFloat, defaultValue=0, description=""}
  User.Delay {type=NiagaraFloat, defaultValue=0.4, description=""}
  User.EffectScale {type=NiagaraFloat, defaultValue=1, description=""}
  User.EffectSpawnScale {type=NiagaraFloat, defaultValue=10, description=""}
  User.ForwardSpeed {type=NiagaraFloat, defaultValue=0, description=""}
  User.FrontPontoonBig.Depth {type=NiagaraFloat, defaultValue=0, description=""}
  User.FrontPontoonBig.Impulse {type=Vector3f, description="", defaultValue {struct=Vector3f, value=<0 0 0>}}
  User.GlobalOffset {type=Vector3f, description="", defaultValue {struct=Vector3f, value=<0 0 0>}}
  User.Graffiti_Color_Array {
   type=NiagaraDataInterfaceArrayColor
   description=""
   defaultValue {
    struct=NiagaraExt_VariableValue_DataInterface
    value {
     dataInterface=&self:NiagaraDataInterfaceArrayColor_0
     dataInterfaceClass=&NiagaraDataInterfaceArrayColor
    }
   }
  }
  User.InWater {type=NiagaraBool}
  User.isBoosting {type=NiagaraBool}
  User.MidPontoonBig.Depth {type=NiagaraFloat, defaultValue=0, description=""}
  User.MidPontoonBig.Impulse {type=Vector3f, description="", defaultValue {struct=Vector3f, value=<0 0 0>}}
  User.NewVariable {type=NiagaraFloat, defaultValue=0, description=""}
  User.RearPontoonBig.Depth {type=NiagaraFloat, defaultValue=0, description=""}
  User.RearPontoonBig.Impulse {type=Vector3f, description="", defaultValue {struct=Vector3f, value=<0 0 0>}}
  User.Scale_Rate {type=NiagaraFloat, defaultValue=0.45, description=""}
  User.Speed {type=NiagaraFloat, defaultValue=1, description=""}
  User.WaterBodyIndex {type=NiagaraFloat, defaultValue=0, description=""}
  User.WaterDepth {type=NiagaraFloat, defaultValue=0, description=""}
  User.WaterPlaneLocation {type=Vector3f, description="", defaultValue {struct=Vector3f, value=<0 0 0>}}
  User.WaterPlaneNormal {type=Vector3f, description="", defaultValue {struct=Vector3f, value=<0 0 0>}}
  User.WaveTime {type=NiagaraFloat, defaultValue=0, description=""}
 ]
}
